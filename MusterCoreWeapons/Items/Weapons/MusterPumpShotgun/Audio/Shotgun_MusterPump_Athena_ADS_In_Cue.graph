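# Sound cue for the Muster pump shotgun's aim-down-sights-in foley.
# A Random node picks between two equally weighted wave variants, feeding a Modulator at the root.
# The cue plays quietly in the weapon foley sound class and borrows the weapon reload attenuation settings for spatial falloff.
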SoundCue /MusterCoreWeapons/Items/Weapons/MusterPumpShotgun/Audio/Shotgun_MusterPump_Athena_ADS_In_Cue
root=Modulator_0
nodes [
 {id=Modulator_0, children [Random_1]}
 {id=Random_1, children [WavePlayer_4,WavePlayer_5], properties {weights [1,1]}}
 {id=WavePlayer_4, properties {soundWaveAssetPtr=&/Shotgun_MusterPump_Athena_ADS_In_01}}
 {id=WavePlayer_5, properties {soundWaveAssetPtr=&/Shotgun_MusterPump_Athena_ADS_In_02}}
]
properties {
 attenuationSettings=&Fort_Weapon_Reload_Att
 maxDistance=2250
 soundClassObject=&Fort_WeaponFoley
 volumeMultiplier=0.2
}
