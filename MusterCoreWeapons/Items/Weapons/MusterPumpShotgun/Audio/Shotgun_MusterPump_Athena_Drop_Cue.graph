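# Sound cue for dropping the Muster pump shotgun in the game world.
# A random node picks one of three drop waves with equal weight.
# It plays through weapon-drop attenuation in the loot collect sound class.
# A shared stop-oldest weapon drop concurrency set governs overlapping instances, alongside a prevent-new override value.
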SoundCue /MusterCoreWeapons/Items/Weapons/MusterPumpShotgun/Audio/Shotgun_MusterPump_Athena_Drop_Cue
root=Random_0
nodes [
 {id=Random_0, children [WavePlayer_0,WavePlayer_5,WavePlayer_6], properties {weights [1,1,1]}}
 {id=WavePlayer_0, properties {soundWaveAssetPtr=&/Shotgun_MusterPump_Athena_Drop_01}}
 {id=WavePlayer_5, properties {soundWaveAssetPtr=&/Shotgun_MusterPump_Athena_Drop_02}}
 {id=WavePlayer_6, properties {soundWaveAssetPtr=&/Shotgun_MusterPump_Athena_Drop_03}}
]
properties {
 attenuationSettings=&Fort_Weapon_Drop_Att
 maxDistance=3000
 soundClassObject=&Fort_Loot_Collect
 volumeMultiplier=1
 concurrencyOverrides {resolutionRule=PreventNew}
 concurrencySet [&Weapon_Drop_StopOldest]
}
