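# Sound cue for the first-person fire sound of the Muster pump shotgun, with a trailing shell casing sound.
# A mixer layers a modulated random pick of six fire waves with a quieter random pick of ten casing waves, modulated and delayed by a fixed interval after the shot.
# Both random nodes are equally weighted and excluded from branch culling.
# It uses first-person weapon sound class, attenuation and gunfire concurrency assets.
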SoundCue /MusterCoreWeapons/Items/Weapons/MusterPumpShotgun/Audio/Shotgun_MusterPump_Athena_Fire_1p_Cue
root=Mixer_0
nodes [
 {id=Delay_0, children [Modulator_1], properties {delayMax=1.3, delayMin=1.3}}
 {id=Mixer_0, children [Modulator_0,Delay_0], properties {inputVolume [1,0.3]}}
 {id=Modulator_0, children [Random_1]}
 {id=Modulator_1, children [Random_0], properties {volumeMax=1}}
 {
  id=Random_0
  children [
   WavePlayer_0
   WavePlayer_5
   WavePlayer_6
   WavePlayer_7
   WavePlayer_10
   WavePlayer_11
   WavePlayer_12
   WavePlayer_13
   WavePlayer_14
   WavePlayer_15
  ]
  properties {bShouldExcludeFromBranchCulling=true, weights [1,1,1,1,1,1,1,1,1,1]}
 }
 {
  id=Random_1
  children [WavePlayer_1,WavePlayer_2,WavePlayer_3,WavePlayer_4,WavePlayer_8,WavePlayer_9]
  properties {bShouldExcludeFromBranchCulling=true, weights [1,1,1,1,1,1]}
 }
 {
  id=WavePlayer_0
  properties {soundWaveAssetPtr=&/Shotgun_MusterPump_Athena_ShellCasing_01}
 }
 {
  id=WavePlayer_1
  properties {soundWaveAssetPtr=&/Shotgun_MusterPump_Athena_Fire_1p_01}
 }
 {
  id=WavePlayer_10
  properties {soundWaveAssetPtr=&/Shotgun_MusterPump_Athena_ShellCasing_05}
 }
 {
  id=WavePlayer_11
  properties {soundWaveAssetPtr=&/Shotgun_MusterPump_Athena_ShellCasing_06}
 }
 {
  id=WavePlayer_12
  properties {soundWaveAssetPtr=&/Shotgun_MusterPump_Athena_ShellCasing_07}
 }
 {
  id=WavePlayer_13
  properties {soundWaveAssetPtr=&/Shotgun_MusterPump_Athena_ShellCasing_08}
 }
 {
  id=WavePlayer_14
  properties {soundWaveAssetPtr=&/Shotgun_MusterPump_Athena_ShellCasing_09}
 }
 {
  id=WavePlayer_15
  properties {soundWaveAssetPtr=&/Shotgun_MusterPump_Athena_ShellCasing_10}
 }
 {
  id=WavePlayer_2
  properties {soundWaveAssetPtr=&/Shotgun_MusterPump_Athena_Fire_1p_02}
 }
 {
  id=WavePlayer_3
  properties {soundWaveAssetPtr=&/Shotgun_MusterPump_Athena_Fire_1p_03}
 }
 {
  id=WavePlayer_4
  properties {soundWaveAssetPtr=&/Shotgun_MusterPump_Athena_Fire_1p_04}
 }
 {
  id=WavePlayer_5
  properties {soundWaveAssetPtr=&/Shotgun_MusterPump_Athena_ShellCasing_02}
 }
 {
  id=WavePlayer_6
  properties {soundWaveAssetPtr=&/Shotgun_MusterPump_Athena_ShellCasing_03}
 }
 {
  id=WavePlayer_7
  properties {soundWaveAssetPtr=&/Shotgun_MusterPump_Athena_ShellCasing_04}
 }
 {
  id=WavePlayer_8
  properties {soundWaveAssetPtr=&/Shotgun_MusterPump_Athena_Fire_1p_05}
 }
 {
  id=WavePlayer_9
  properties {soundWaveAssetPtr=&/Shotgun_MusterPump_Athena_Fire_1p_06}
 }
]
properties {
 attenuationSettings=&Fort_Weapon_Projectile_FirstPerson_Athena_Att
 bDebug=true
 maxDistance=2097152
 soundClassObject=&Fort_WeaponRanged_Fire_1P_PSM
 volumeMultiplier=1.4
 attenuationOverrides {manualReverbSendLevel=-1.0165, reverbSendMethod=Manual}
 concurrencyOverrides {bLimitToOwner=true, maxCount=3, resolutionRule=StopOldest, volumeScale=0.9}
 concurrencySet [&Weapon_Gunfire_FirstPerson_Conc]
}
