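# Sound cue for the Muster pump shotgun's fire sound, layering gunshot and shell casing audio.
# Close, medium and distant shots each pick a random variant through their own attenuation and sound class, with medium and distant sharing the distant class.
# A randomized shell-casing layer plays after a fixed delay as weapon foley with reload attenuation.
# Nested mixers balance the layers, and concurrency limits instances per owner, stopping the oldest.
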SoundCue /MusterCoreWeapons/Items/Weapons/MusterPumpShotgun/Audio/Shotgun_MusterPump_Athena_Fire_Cue
root=Mixer_2
nodes [
 {id=Attenuation_0, children [Delay_0], properties {attenuationSettings=&Fort_Weapon_Reload_Att}}
 {id=Attenuation_3, children [Modulator_4], properties {attenuationSettings=&Fort_Weapon_Projectile_Close_Athena_Att}}
 {
  id=Attenuation_4
  children [Modulator_0]
  properties {attenuationSettings=&Fort_Weapon_Projectile_Distant_Mono_Athena_Att}
 }
 {
  id=Attenuation_5
  children [Modulator_2]
  properties {attenuationSettings=&Fort_Weapon_Projectile_Medium_Mono_Athena_Att}
 }
 {id=Delay_0, children [Modulator_1], properties {delayMax=1.3, delayMin=1.3}}
 {id=Mixer_2, children [Modulator_3,SoundClass_4], properties {inputVolume [1,0.15]}}
 {id=Mixer_3, children [SoundClass_0,SoundClass_3,SoundClass_5], properties {inputVolume [0.3,1.2,0.4]}}
 {id=Modulator_0, children [Random_2], properties {volumeMax=2, volumeMin=2}}
 {id=Modulator_1, children [Random_4], properties {volumeMax=1}}
 {id=Modulator_2, children [Random_1], properties {volumeMax=2, volumeMin=2}}
 {id=Modulator_3, children [Mixer_3]}
 {id=Modulator_4, children [Random_0], properties {volumeMax=1, volumeMin=1}}
 {
  id=Random_0
  children [WavePlayer_0,WavePlayer_1,WavePlayer_2,WavePlayer_7,WavePlayer_8,WavePlayer_9]
  properties {bShouldExcludeFromBranchCulling=true, weights [1,1,1,1,1,1]}
 }
 {
  id=Random_1
  children [WavePlayer_10, WavePlayer_11, WavePlayer_12, WavePlayer_19, WavePlayer_20, WavePlayer_21]
  properties {bShouldExcludeFromBranchCulling=true, weights [1,1,1,1,1,1]}
 }
 {
  id=Random_2
  children [WavePlayer_22, WavePlayer_23, WavePlayer_24, WavePlayer_25, WavePlayer_26, WavePlayer_27]
  properties {bShouldExcludeFromBranchCulling=true, weights [1,1,1,1,1,1]}
 }
 {
  id=Random_4
  children [
   WavePlayer_15
   WavePlayer_5
   WavePlayer_6
   WavePlayer_16
   WavePlayer_17
   WavePlayer_18
   WavePlayer_28
   WavePlayer_13
   WavePlayer_14
   WavePlayer_29
  ]
  properties {bShouldExcludeFromBranchCulling=true, weights [1,1,1,1,1,1,1,1,1,1]}
 }
 {id=SoundClass_0, children [Attenuation_0], properties {soundClassOverride=&Fort_WeaponFoley}}
 {
  id=SoundClass_3
  children [Attenuation_3]
  properties {soundClassOverride=&Fort_WeaponRanged_Fire_3P_PSM}
 }
 {
  id=SoundClass_4
  children [Attenuation_4]
  properties {soundClassOverride=&Fort_WeaponRanged_Fire_Distant}
 }
 {
  id=SoundClass_5
  children [Attenuation_5]
  properties {soundClassOverride=&Fort_WeaponRanged_Fire_Distant}
 }
 {
  id=WavePlayer_0
  properties {soundWaveAssetPtr=&/Shotgun_MusterPump_Athena_Fire_Close_01}
 }
 {
  id=WavePlayer_1
  properties {soundWaveAssetPtr=&/Shotgun_MusterPump_Athena_Fire_Close_02}
 }
 {
  id=WavePlayer_10
  properties {soundWaveAssetPtr=&/Shotgun_MusterPump_Athena_Fire_Med_01}
 }
 {
  id=WavePlayer_11
  properties {soundWaveAssetPtr=&/Shotgun_MusterPump_Athena_Fire_Med_02}
 }
 {
  id=WavePlayer_12
  properties {soundWaveAssetPtr=&/Shotgun_MusterPump_Athena_Fire_Med_03}
 }
 {
  id=WavePlayer_13
  properties {soundWaveAssetPtr=&/Shotgun_MusterPump_Athena_ShellCasing_08}
 }
 {
  id=WavePlayer_14
  properties {soundWaveAssetPtr=&/Shotgun_MusterPump_Athena_ShellCasing_09}
 }
 {
  id=WavePlayer_15
  properties {soundWaveAssetPtr=&/Shotgun_MusterPump_Athena_ShellCasing_01}
 }
 {
  id=WavePlayer_16
  properties {soundWaveAssetPtr=&/Shotgun_MusterPump_Athena_ShellCasing_04}
 }
 {
  id=WavePlayer_17
  properties {soundWaveAssetPtr=&/Shotgun_MusterPump_Athena_ShellCasing_05}
 }
 {
  id=WavePlayer_18
  properties {soundWaveAssetPtr=&/Shotgun_MusterPump_Athena_ShellCasing_06}
 }
 {
  id=WavePlayer_19
  properties {soundWaveAssetPtr=&/Shotgun_MusterPump_Athena_Fire_Med_04}
 }
 {
  id=WavePlayer_2
  properties {soundWaveAssetPtr=&/Shotgun_MusterPump_Athena_Fire_Close_03}
 }
 {
  id=WavePlayer_20
  properties {soundWaveAssetPtr=&/Shotgun_MusterPump_Athena_Fire_Med_05}
 }
 {
  id=WavePlayer_21
  properties {soundWaveAssetPtr=&/Shotgun_MusterPump_Athena_Fire_Med_06}
 }
 {
  id=WavePlayer_22
  properties {soundWaveAssetPtr=&/Shotgun_MusterPump_Athena_Fire_Dist_06}
 }
 {
  id=WavePlayer_23
  properties {soundWaveAssetPtr=&/Shotgun_MusterPump_Athena_Fire_Dist_01}
 }
 {
  id=WavePlayer_24
  properties {soundWaveAssetPtr=&/Shotgun_MusterPump_Athena_Fire_Dist_02}
 }
 {
  id=WavePlayer_25
  properties {soundWaveAssetPtr=&/Shotgun_MusterPump_Athena_Fire_Dist_03}
 }
 {
  id=WavePlayer_26
  properties {soundWaveAssetPtr=&/Shotgun_MusterPump_Athena_Fire_Dist_04}
 }
 {
  id=WavePlayer_27
  properties {soundWaveAssetPtr=&/Shotgun_MusterPump_Athena_Fire_Dist_05}
 }
 {
  id=WavePlayer_28
  properties {soundWaveAssetPtr=&/Shotgun_MusterPump_Athena_ShellCasing_07}
 }
 {
  id=WavePlayer_29
  properties {soundWaveAssetPtr=&/Shotgun_MusterPump_Athena_ShellCasing_10}
 }
 {
  id=WavePlayer_5
  properties {soundWaveAssetPtr=&/Shotgun_MusterPump_Athena_ShellCasing_02}
 }
 {
  id=WavePlayer_6
  properties {soundWaveAssetPtr=&/Shotgun_MusterPump_Athena_ShellCasing_03}
 }
 {
  id=WavePlayer_7
  properties {soundWaveAssetPtr=&/Shotgun_MusterPump_Athena_Fire_Close_04}
 }
 {
  id=WavePlayer_8
  properties {soundWaveAssetPtr=&/Shotgun_MusterPump_Athena_Fire_Close_05}
 }
 {
  id=WavePlayer_9
  properties {soundWaveAssetPtr=&/Shotgun_MusterPump_Athena_Fire_Close_06}
 }
]
properties {
 bDebug=true
 maxDistance=27500
 soundClassObject=&Fort_WeaponRanged_Fire_3P_PSM
 volumeMultiplier=0.451
 concurrencyOverrides {bLimitToOwner=true, maxCount=3, resolutionRule=StopOldest}
 concurrencySet [&Weapon_Gunfire_Global_Conc, &Weapon_Gunfire_Local_Conc]
}
