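# Pickup sound cue for the Muster Pump Shotgun, using weapon-pickup attenuation and the loot-collect sound class.
# A quality-level root sends its first two levels to one mixer and its last two to another.
# Each mixer slightly boosts a delay-wrapped wave player for a Pickup_01 wave and leaves its second input empty.
# The cue lowers overall volume and references the PickupMaximum_Oldest concurrency set.
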SoundCue /MusterCoreWeapons/Items/Weapons/MusterPumpShotgun/Audio/Shotgun_MusterPump_Athena_Pickup_Cue
root=QualityLevel_0
nodes [
 {id=Delay_0, children [WavePlayer_3]}
 {id=Delay_1, children [WavePlayer_2]}
 {id=Mixer_0, children [Delay_0,null], properties {inputVolume [1.2,1]}}
 {id=Mixer_1, children [Delay_1,null], properties {inputVolume [1.2,1]}}
 {id=QualityLevel_0, children [Mixer_1,Mixer_1,Mixer_0,Mixer_0]}
 {id=WavePlayer_2, properties {soundWaveAssetPtr=&/Shotgun_MusterPump_Athena_Pickup_01}}
 {id=WavePlayer_3, properties {soundWaveAssetPtr=&Shotgun_MusterPump_Athena_Pickup_01}}
]
properties {
 attenuationSettings=&Fort_Weapon_Pickup_Att
 maxDistance=3500
 soundClassObject=&Fort_Loot_Collect
 volumeMultiplier=0.4615
 concurrencyOverrides {maxCount=1, resolutionRule=PreventNew}
 concurrencySet [&PickupMaximum_Oldest]
}
