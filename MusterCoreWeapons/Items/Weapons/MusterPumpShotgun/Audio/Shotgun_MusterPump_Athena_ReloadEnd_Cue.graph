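# Sound cue for the end of the reload on the Muster pump shotgun, routed to the weapon foley sound class.
# A random node picks one of six equally weighted ReloadEnd wave variations, and a modulator above it adds slight pitch variation at constant volume.
# The cue uses shared weapon reload attenuation at reduced volume, overrides concurrency, and has its debug flag enabled.
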SoundCue /MusterCoreWeapons/Items/Weapons/MusterPumpShotgun/Audio/Shotgun_MusterPump_Athena_ReloadEnd_Cue
root=Modulator_0
nodes [
 {id=Modulator_0, children [Random_0], properties {pitchMax=1.01, pitchMin=0.98, volumeMax=1, volumeMin=1}}
 {id=Random_0, children [WavePlayer_0,WavePlayer_1,WavePlayer_2,WavePlayer_3,WavePlayer_4,WavePlayer_5], properties {weights [1,1,1,1,1,1]}}
 {
  id=WavePlayer_0
  properties {soundWaveAssetPtr=&/Shotgun_MusterPump_Athena_ReloadEnd_01}
 }
 {
  id=WavePlayer_1
  properties {soundWaveAssetPtr=&/Shotgun_MusterPump_Athena_ReloadEnd_02}
 }
 {
  id=WavePlayer_2
  properties {soundWaveAssetPtr=&/Shotgun_MusterPump_Athena_ReloadEnd_03}
 }
 {
  id=WavePlayer_3
  properties {soundWaveAssetPtr=&/Shotgun_MusterPump_Athena_ReloadEnd_04}
 }
 {
  id=WavePlayer_4
  properties {soundWaveAssetPtr=&/Shotgun_MusterPump_Athena_ReloadEnd_05}
 }
 {
  id=WavePlayer_5
  properties {soundWaveAssetPtr=&/Shotgun_MusterPump_Athena_ReloadEnd_06}
 }
]
properties {
 attenuationSettings=&Fort_Weapon_Reload_Att
 bDebug=true
 bOverrideConcurrency=true
 maxDistance=2250
 soundClassObject=&Fort_WeaponFoley
 volumeMultiplier=0.551
}
