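# Sound cue for the start of the Muster pump shotgun's reload, classed as weapon foley.
# A Random node picks one of two equally weighted wave variants and feeds a Modulator that adds slight pitch variation at constant volume.
# It uses the shared weapon-reload attenuation, plays at reduced volume, overrides concurrency, and has its debug flag switched on.
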SoundCue /MusterCoreWeapons/Items/Weapons/MusterPumpShotgun/Audio/Shotgun_MusterPump_Athena_ReloadStart_Cue
root=Modulator_0
nodes [
 {id=Modulator_0, children [Random_0], properties {pitchMax=1.02, pitchMin=0.98, volumeMax=1, volumeMin=1}}
 {id=Random_0, children [WavePlayer_0,WavePlayer_1], properties {weights [1,1]}}
 {
  id=WavePlayer_0
  properties {soundWaveAssetPtr=&/Shotgun_MusterPump_Athena_ReloadStart_01}
 }
 {
  id=WavePlayer_1
  properties {soundWaveAssetPtr=&/Shotgun_MusterPump_Athena_ReloadStart_02}
 }
]
properties {
 attenuationSettings=&Fort_Weapon_Reload_Att
 bDebug=true
 bOverrideConcurrency=true
 maxDistance=2250
 soundClassObject=&Fort_WeaponFoley
 volumeMultiplier=0.451
}
